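# Idle Niagara effect for the SaladDressing pickaxe, drawing a tapered laser shape from two nested ribbon emitters, main and inner.
# Each burst-spawns a few non-expiring particles, re-placed every update along local Y from a sampled skeletal mesh transform.
# Rotated up, down and side offsets with facing vectors give multiple ribbon renderers a volumetric look.
# The main laser adds a glow width driven by camera angle.
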NiagaraSystem /BRCosmetics/Weapons/FORT_Melee/Pickaxe_SaladDressing/FX/NS_Pickaxe_SaladDressing_Idle
dependencies {
 usedDataInterfaces [&NiagaraDataInterfaceCurve, &NiagaraDataInterfaceCamera, &NiagaraDataInterfaceSkeletalMesh]
 usedDynamicInputs []
 usedModules [
  &self:SystemSpawnScript.SystemScriptSource.SystemScriptGraph.NiagaraNodeAssignment_0.SetVariables_#1
  &SystemState
 ]
}
emitters {
 NE_InnerLaser {
  emitterSpawnScript [
   /SetVariables(Emitter.ParticleInRibbon=5) {isSetParameters=true, module=SetVariables_#2}
   /SetVariables(Emitter.SideRibbonOffset=0.05) {isSetParameters=true}
   /SetVariables(Emitter.UpRibbonOffset=0.2) {isSetParameters=true}
  ]
  emitterUpdateScript [
   &EmitterState(
    Inactive Response=::NewEnumerator1
    Life Cycle Mode=::NewEnumerator1
    Loop Behavior=::NewEnumerator1
    Loop Duration Mode=::NewEnumerator1
    UseLoopDelay=false
   )
   &SpawnBurst_Instantaneous(Spawn Count=$Emitter.ParticleInRibbon, Use Spawn Probability=false, UseLoopCountLimit=false)
  ]
  particleSpawnScript [
   &InitializeParticle(
    Color Mode=::NewEnumerator0
    Lifetime=5
    Ribbon Width=10
    Ribbon Width Mode=::NewEnumerator1
    Sprite Size Mode=::NewEnumerator3
    UsePositionOffset=false
    Write Material Random=true
   )
   &PartitionParticles(
    Normalized Index In Partition Includes One=true
    Number Of Partitions=1
    Particles Per Partition:NiagaraInt32=$Emitter.ParticleInRibbon
    Partition Mode:ENiagara_PartitionMode=::NewEnumerator1
   )
   /SetVariables(Particles.RibbonU0Override=$Particles.PartitionParticles.Normalized Index In Partition) {isSetParameters=true, module=SetVariables_#3}
  ]
  particleUpdateScript [
   &ParticleState(Kill Particles When Lifetime Has Elapsed=false)
   &SkeletalMeshLocation(
    MaskSpawnGroup=false
    Mesh Sampling Type=::NewEnumerator5
    Sampled Position Offset=&Lerp_Vector(A=<0 0 1>, B=<0 50 0>, Alpha=$Particles.PartitionParticles.Normalized Index In Partition)
    Skeletal Mesh=$System.SkeletalMesh
    Socket Sampling Mode=::NewEnumerator5
   )
   /SetVariables(Particles.TopBotWidth:NiagaraFloat=&Multiply_Float(
     A=3
     B=&FloatFromCurve(
      FloatCurve=DataInterface() {propertyValues {Curve {keys [<at 0 1>,<at 1 -0.2>]}}}
      CurveIndex=$Particles.PartitionParticles.Normalized Index In Partition
     )
    )) {isSetParameters=true, module=SetVariables_#4}
   /SetVariables(Particles.RibbonWidth=&Multiply_Float(
     A=10
     B=&FloatFromCurve(
      FloatCurve=DataInterface() {propertyValues {Curve {keys [<at 0 1>,<at 1 0.2>]}}}
      CurveIndex=$Particles.PartitionParticles.Normalized Index In Partition
     )
    )) {isSetParameters=true, module=SetVariables_#5}
   /SetVariables(
    Particles.InvertedRibbonFacing:Vector3f=&RotateVectorByQuaternion(Vector To Rotate=<0 0 -1>, Quaternion=$Particles.SkeletalMeshLocation.SampledRotation)
    Particles.RibbonFacing=&RotateVectorByQuaternion(Vector To Rotate=<0 0 1>, Quaternion=$Particles.SkeletalMeshLocation.SampledRotation)
   ) {isSetParameters=true, module=SetVariables_#6}
   /SetVariables(
    Particles.BotRibbonPos:NiagaraPosition=&AddVectorToPosition(
     Position=$Particles.Position
     Vector=&RotateVectorByQuaternion(
      Vector To Rotate=&MakeVector(
       X=0
       Y=0
       Z=&Multiply_Float(A=&Multiply_Float(A=$Emitter.UpRibbonOffset, B=$Particles.TopBotWidth), B=-1)
      )
      Quaternion=$Particles.SkeletalMeshLocation.SampledRotation
     )
    )
    Particles.LeftRibbonPos:NiagaraPosition=&AddVectorToPosition(
     Position=$Particles.Position
     Vector=&RotateVectorByQuaternion(
      Vector To Rotate=&MakeVector(
       X=&Multiply_Float(A=&Multiply_Float(A=$Emitter.SideRibbonOffset, B=$Particles.RibbonWidth), B=-1)
       Y=0
       Z=0
      )
      Quaternion=$Particles.SkeletalMeshLocation.SampledRotation
     )
    )
    Particles.RightRibbonPos:NiagaraPosition=&AddVectorToPosition(
     Position=$Particles.Position
     Vector=&RotateVectorByQuaternion(
      Vector To Rotate=&MakeVector(X=&Multiply_Float(A=$Emitter.SideRibbonOffset, B=$Particles.RibbonWidth), Y=0, Z=0)
      Quaternion=$Particles.SkeletalMeshLocation.SampledRotation
     )
    )
    Particles.TopRibbonPos:NiagaraPosition=&AddVectorToPosition(
     Position=$Particles.Position
     Vector=&RotateVectorByQuaternion(
      Vector To Rotate=&MakeVector(X=0, Y=0, Z=&Multiply_Float(A=$Emitter.UpRibbonOffset, B=$Particles.TopBotWidth))
      Quaternion=$Particles.SkeletalMeshLocation.SampledRotation
     )
    )
   ) {isSetParameters=true}
   &DynamicMaterialParameters(
    Index 1 Param 3=&FloatFromCurve(
     FloatCurve=DataInterface() {propertyValues {Curve {keys [<at 0.5 0>,<at 1 1>]}}}
     CurveIndex=$Particles.PartitionParticles.Normalized Index In Partition
    )
    Module.Index0Param0Write=true
    Module.Index0Param1Write=true
    Module.Index0Param2Write=true
    Module.Index0Param3Write=true
    Module.Index1Param0Write=true
    Module.Index1Param1Write=true
    Module.Index1Param2Write=true
    Module.Index1Param3Write=true
    Write Parameter Index 1=true
   )
  ]
  rendererClasses [&NiagaraRibbonRendererProperties]
  renderers [
   {rendererClass=&NiagaraRibbonRendererProperties}
   {rendererClass=&NiagaraRibbonRendererProperties, rendererIndex=1}
   {rendererClass=&NiagaraRibbonRendererProperties, rendererIndex=2}
   {rendererClass=&NiagaraRibbonRendererProperties, rendererIndex=3}
  ]
 }
 NE_MainLaser {
  emitterSpawnScript [
   /SetVariables(Emitter.ParticleInRibbon=10) {isSetParameters=true, module=SetVariables_#2}
   /SetVariables(Emitter.SideRibbonOffset=0.05) {isSetParameters=true}
   /SetVariables(Emitter.UpRibbonOffset=0.2) {isSetParameters=true}
  ]
  emitterUpdateScript [
   &EmitterState(
    Inactive Response=::NewEnumerator1
    Life Cycle Mode=::NewEnumerator1
    Loop Behavior=::NewEnumerator1
    Loop Duration Mode=::NewEnumerator1
    UseLoopDelay=false
   )
   &SpawnBurst_Instantaneous(Spawn Count=$Emitter.ParticleInRibbon, Use Spawn Probability=false, UseLoopCountLimit=false)
  ]
  particleSpawnScript [
   &InitializeParticle(
    Color Mode=::NewEnumerator0
    Lifetime=5
    Ribbon Width=10
    Ribbon Width Mode=::NewEnumerator1
    Sprite Size Mode=::NewEnumerator3
    UsePositionOffset=false
    Write Material Random=true
   )
   &PartitionParticles(
    Normalized Index In Partition Includes One=true
    Number Of Partitions=1
    Particles Per Partition:NiagaraInt32=$Emitter.ParticleInRibbon
    Partition Mode:ENiagara_PartitionMode=::NewEnumerator1
   )
   /SetVariables(Particles.RibbonU0Override=$Particles.PartitionParticles.Normalized Index In Partition) {isSetParameters=true, module=SetVariables_#3}
  ]
  particleUpdateScript [
   &ParticleState(Kill Particles When Lifetime Has Elapsed=false)
   &SkeletalMeshLocation(
    MaskSpawnGroup=false
    Mesh Sampling Type=::NewEnumerator5
    Sampled Position Offset=&Lerp_Vector(A=<0 0 1>, B=<0 50 0>, Alpha=$Particles.PartitionParticles.Normalized Index In Partition)
    Skeletal Mesh=$System.SkeletalMesh
    Socket Sampling Mode=::NewEnumerator5
   )
   /SetVariables(Particles.TopBotWidth:NiagaraFloat=&Multiply_Float(
     A=10
     B=&FloatFromCurve(
      FloatCurve=DataInterface() {propertyValues {Curve {keys [<at 0 1>,<at 1 -0.2>]}}}
      CurveIndex=$Particles.PartitionParticles.Normalized Index In Partition
     )
    )) {isSetParameters=true, module=SetVariables_#4}
   /SetVariables(Particles.RibbonWidth=&Multiply_Float(
     A=23
     B=&FloatFromCurve(
      FloatCurve=DataInterface() {propertyValues {Curve {keys [<at 0 1>,<at 1 0.2>]}}}
      CurveIndex=$Particles.PartitionParticles.Normalized Index In Partition
     )
    )) {isSetParameters=true, module=SetVariables_#5}
   /SetVariables(Particles.GlowWidth:NiagaraFloat=&Multiply_Float(
     A=&Lerp_Float(
      A=13
      B=32
      Alpha=&Abs_Float(Float:NiagaraFloat=&DotProduct(
        Vector A=&ReturnCameraVector(Camera Vectors=::NewEnumerator1, Coordinate Space=::Simulation)
        Vector B=&RotateVectorByQuaternion(Vector To Rotate=<1 0 0>, Quaternion=$Particles.SkeletalMeshLocation.SampledRotation)
       ))
     )
     B=&FloatFromCurve(
      FloatCurve=DataInterface() {propertyValues {Curve {keys [
          {interpMode=RCIM_Cubic, tangentMode=RCTM_User, time=0.1, value=0.4}
          {arriveTangent=0.2, interpMode=RCIM_Cubic, leaveTangent=0.2, tangentMode=RCTM_User, time=0.2, value=1}
          {interpMode=RCIM_Cubic, tangentMode=RCTM_User, time=1, value=0.2}
         ]}}}
      CurveIndex=$Particles.PartitionParticles.Normalized Index In Partition
     )
    )) {isSetParameters=true}
   /SetVariables(
    Particles.InvertedRibbonFacing:Vector3f=&RotateVectorByQuaternion(Vector To Rotate=<0 0 -1>, Quaternion=$Particles.SkeletalMeshLocation.SampledRotation)
    Particles.RibbonFacing=&RotateVectorByQuaternion(Vector To Rotate=<0 0 1>, Quaternion=$Particles.SkeletalMeshLocation.SampledRotation)
   ) {isSetParameters=true, module=SetVariables_#6}
   /SetVariables(
    Particles.BotRibbonPos:NiagaraPosition=&AddVectorToPosition(
     Position=$Particles.Position
     Vector=&RotateVectorByQuaternion(
      Vector To Rotate=&MakeVector(
       X=0
       Y=0
       Z=&Multiply_Float(A=&Multiply_Float(A=$Emitter.UpRibbonOffset, B=$Particles.TopBotWidth), B=-1)
      )
      Quaternion=$Particles.SkeletalMeshLocation.SampledRotation
     )
    )
    Particles.LeftRibbonPos:NiagaraPosition=&AddVectorToPosition(
     Position=$Particles.Position
     Vector=&RotateVectorByQuaternion(
      Vector To Rotate=&MakeVector(
       X=&Multiply_Float(A=&Multiply_Float(A=$Emitter.SideRibbonOffset, B=$Particles.RibbonWidth), B=-1)
       Y=0
       Z=0
      )
      Quaternion=$Particles.SkeletalMeshLocation.SampledRotation
     )
    )
    Particles.RightRibbonPos:NiagaraPosition=&AddVectorToPosition(
     Position=$Particles.Position
     Vector=&RotateVectorByQuaternion(
      Vector To Rotate=&MakeVector(X=&Multiply_Float(A=$Emitter.SideRibbonOffset, B=$Particles.RibbonWidth), Y=0, Z=0)
      Quaternion=$Particles.SkeletalMeshLocation.SampledRotation
     )
    )
    Particles.TopRibbonPos:NiagaraPosition=&AddVectorToPosition(
     Position=$Particles.Position
     Vector=&RotateVectorByQuaternion(
      Vector To Rotate=&MakeVector(X=0, Y=0, Z=&Multiply_Float(A=$Emitter.UpRibbonOffset, B=$Particles.TopBotWidth))
      Quaternion=$Particles.SkeletalMeshLocation.SampledRotation
     )
    )
   ) {isSetParameters=true}
   &DynamicMaterialParameters(
    Index 1 Param 3=&FloatFromCurve(
     FloatCurve=DataInterface() {propertyValues {Curve {keys [<at 0.5 0>,<at 1 1>]}}}
     CurveIndex=$Particles.PartitionParticles.Normalized Index In Partition
    )
    Module.Index0Param0Write=true
    Module.Index0Param1Write=true
    Module.Index0Param2Write=true
    Module.Index0Param3Write=true
    Module.Index1Param0Write=true
    Module.Index1Param1Write=true
    Module.Index1Param2Write=true
    Module.Index1Param3Write=true
    Write Parameter Index 1=true
   )
  ]
  rendererClasses [&NiagaraRibbonRendererProperties]
  renderers [
   {rendererClass=&NiagaraRibbonRendererProperties}
   {rendererClass=&NiagaraRibbonRendererProperties, rendererIndex=1}
   {rendererClass=&NiagaraRibbonRendererProperties, rendererIndex=2}
   {rendererClass=&NiagaraRibbonRendererProperties, rendererIndex=3}
   {rendererClass=&NiagaraRibbonRendererProperties, rendererIndex=4}
  ]
 }
}
summary {
 systemName=NS_Pickaxe_SaladDressing_Idle
 emitters [
  {emitterName=NE_MainLaser, rendererClasses [&NiagaraRibbonRendererProperties]}
  {emitterName=NE_InnerLaser, rendererClasses [&NiagaraRibbonRendererProperties]}
 ]
 userVariables []
}
