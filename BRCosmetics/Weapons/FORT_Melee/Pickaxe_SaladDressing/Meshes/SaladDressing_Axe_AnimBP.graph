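# Animation Blueprint for the SaladDressing pickaxe that adds procedural physics motion to its dyn_cable bone.
# The input pose passes through one Anim Dynamics node to the output, using scaled component motion, a custom sim-space gravity, damping overrides and angular limits to swing the cable.
# A second, softer Anim Dynamics node in actor space is left unconnected, and the event graph holds only unwired default nodes.
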AnimBlueprint /BRCosmetics/Weapons/FORT_Melee/Pickaxe_SaladDressing/Meshes/SaladDressing_Axe_AnimBP
target_skeleton=&/SaladDressing_Axe_Skeleton
graphs [
 AnimGraph {nodes [
   Anim Dynamics - Bone: dyn_cable!(ComponentPose, Alpha=1) -> (Pose) {
    angularDampingOverride=1
    angularSpringConstant=5
    bGravityOverrideInSimSpace=true
    bOverrideAngularDamping=true
    bUseGravityOverride=true
    linearDampingOverride=0.9
    linearSpringConstant=1
    simulationSpace=Actor
    alphaBoolBlend {blendOption=Linear}
    boundBone {boneName=dyn_cable}
    componentAppliedLinearAccClamp=<2500 2500 2500>
    componentLinearAccScale=<0.175 0.15 0.15>
    componentLinearVelScale=<0.175 0.15 0.15>
    gravityOverride=<-250 -250>
    physicsBodyDefinitions [{
      boundBone {boneName=dyn_cable}
      boxExtents=<5 10 10>
      constraintSetup {angularLimitsMax=<0 0 75>, angularLimitsMin=<0 0 -75>, angularTarget=<0 0 0>}
      localJointOffset=<5 0 0>
     }]
    relativeSpaceBone {boneName=Melee_TwoHanded_Handle}
   }
   AnimGraphNode_SubInput_0 = Input Pose()
   n1 = Anim Dynamics - Bone: dyn_cable(ComponentPose=Local To Component(LocalPose=@AnimGraphNode_SubInput_0), Alpha=1) {
    angularDampingOverride=1
    angularSpringConstant=5
    bGravityOverrideInSimSpace=true
    bOverrideAngularDamping=true
    bUseGravityOverride=true
    linearDampingOverride=0.9
    linearSpringConstant=1
    alphaBoolBlend {blendOption=Linear}
    boundBone {boneName=dyn_cable}
    componentAppliedLinearAccClamp=<4000 2500 2500>
    componentLinearAccScale=<0.37 0.2 0.2>
    componentLinearVelScale=<0.3 0.2 0.2>
    gravityOverride=<-600 -400>
    physicsBodyDefinitions [{
      boundBone {boneName=dyn_cable}
      boxExtents=<5 15 15>
      constraintSetup {angularLimitsMax=<0 0 65>, angularLimitsMin=<0 0 -75>, angularTarget=<0 0 0>}
      localJointOffset=<4 0 0>
     }]
    relativeSpaceBone {boneName=Melee_TwoHanded_Handle}
   }
   Output Pose(Result=Component To Local(ComponentPose=@n1)) {layerGroup=None}
  ]}
 EventGraph {nodes [
   TryGetPawnOwner()
   Event BlueprintUpdateAnimation()
  ]}
]
variables []
